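; PCG primitive that recursively splits input areas into smaller ones and outputs the resulting closed-spline Areas plus the splines that divide them.
; Tagged input passes through shared input management, a core split subgraph, then an output-preparation subgraph, with a closed-spline tag filter separating the two outputs.
; Input processing, fallback, core and output stages are swappable graph parameters.
; Parameters tune iteration count, split offset, minimum area, angle jitter, optional disturbance and seed.
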(graph Subdivide_Areas_WithRecursiveSplit
  (params (InputsProcessingGraph :type (SoftObject /Script/PCG.PCGGraphInterface) :default /PCGPrimitives/Primitives/Shared/InputsProcessing/InputsProcessing_ConvertClosedSplinesAndForwardPoly2Ds.InputsProcessing_ConvertClosedSplinesAndForwardPoly2Ds) (InputsFallbackGraph :type (SoftObject /Script/PCG.PCGGraphInterface) :default /PCGPrimitives/Primitives/_Template_/Primitives_InputsFallback.Primitives_InputsFallback) (CoreProcessGraph :type (SoftObject /Script/PCG.PCGGraphInterface) :default /PCGPrimitives/Primitives/Subdivide/Subgraphs/Subdivide_Areas_WithRecursiveSplit_CoreProcess.Subdivide_Areas_WithRecursiveSplit_CoreProcess) (PrepareOutputsGraph :type (SoftObject /Script/PCG.PCGGraphInterface) :default /PCGPrimitives/Primitives/Shared/PrepareOutputs/PrepareOutputs_ClosedSplinesAndSplines.PrepareOutputs_ClosedSplinesAndSplines) (Iterations :type Int :default 2) (MaxSplitDistancFromCenter :type Double :default 0.3) (MinAreaSize :type Double :default 1e+03) (AngleJitter :type Double) (AllowDisturbance :type Bool) (Disturbance :type Double) (DisturbanceSubdivisionsDistance :type Double :default 4e+03) (Seed :type Int :default 3))
  (inputs (PrimaryInput :type Any))
  (outputs (Areas :type Any) (Subdivide%20Splines :type Any))
  (bind AddTags_0 (Add%20Tags :TagsToAdd "Input:0"))
  (bind FilterDataByTag_0 (Filter%20Data%20By%20Tag :SelectedTags "Type:ClosedSpline"))
  ;[0 0 0 0.5 32] Inputs
  (bind Subgraph_0 (Subgraph :Graph /PCGPrimitives/Primitives/_Template_/Subgraphs/Primitives_InputsManagement.Primitives_InputsManagement :InputsFallbackGraph InputsFallbackGraph :InputsProcessingGraph InputsProcessingGraph))
  ;[0 0 0 0.5 32] Core Process
  (bind Subgraph_3 (Subgraph :Graph /PCGPrimitives/Primitives/Subdivide/Subgraphs/Subdivide_Areas_WithRecursiveSplit_CoreProcess.Subdivide_Areas_WithRecursiveSplit_CoreProcess :Allow_Disturbance AllowDisturbance :AngleJitter AngleJitter :Disturbance Disturbance :Disturbance_Subdivisions DisturbanceSubdivisionsDistance :MaxIterations Iterations :MaxSplitDistanceFromCenter MaxSplitDistancFromCenter :MinAreaSize MinAreaSize :Seed-param Seed :SubgraphOverride-param CoreProcessGraph))
  ;[0 0 0 0.5 32] Prepare Output
  (bind Subgraph_4 (Subgraph :Graph /PCGPrimitives/Primitives/Shared/PrepareOutputs/PrepareOutputs_ClosedSplinesAndSplines.PrepareOutputs_ClosedSplinesAndSplines :SubgraphOverride-param PrepareOutputsGraph))
  (connect DefaultInputNode PrimaryInput AddTags_0 In)
  (connect Subgraph_4 Out FilterDataByTag_0 In)
  (connect AddTags_0 Out Subgraph_0 In)
  (connect Subgraph_0 Out Subgraph_3 In)
  (connect Subgraph_3 Out Subgraph_4 In)
  (output Areas FilterDataByTag_0.InsideFilter)
  (output Subdivide%20Splines FilterDataByTag_0.OutsideFilter))
