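; PCG primitive graph that subdivides areas with a grid, set either by cell width and length or by subdivision counts, with optional angle jitter and disturbance.
; Input is tagged, then runs through a chain of input management, core process and output preparation subgraphs.
; Fallback, processing, core and output graphs are exposed as parameters for swapping stages.
; A closed-spline tag filter splits results into Areas and Subdividing_splines outputs.
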(graph Subdivide_Areas_withGrid
  (params (InputsProcessingGraph :type (SoftObject /Script/PCG.PCGGraphInterface) :default /PCGPrimitives/Primitives/Shared/InputsProcessing/InputsProcessing_ConvertPoly2DsAndForwardClosedSplines.InputsProcessing_ConvertPoly2DsAndForwardClosedSplines) (InputsFallbackGraph :type (SoftObject /Script/PCG.PCGGraphInterface) :default /PCGPrimitives/Primitives/_Template_/Primitives_InputsFallback.Primitives_InputsFallback) (CoreProcessGraph :type (SoftObject /Script/PCG.PCGGraphInterface) :default /PCGPrimitives/Primitives/Subdivide/Subgraphs/Subdivide_Areas_withGrid_CoreProcess.Subdivide_Areas_withGrid_CoreProcess) (GridWidth :type Double :default 5e+03) (GridLength :type Double :default 5e+03) (SubdivWidth :type Int :default 1) (SubdivLength :type Int :default 1) (AngleJitter :type Double) (PrepareOutputsGraph :type (SoftObject /Script/PCG.PCGGraphInterface) :default /PCGPrimitives/Primitives/Shared/PrepareOutputs/PrepareOutputs_ClosedSplinesAndSplines.PrepareOutputs_ClosedSplinesAndSplines) (UseSubdivisionInsteadOfSizes :type Bool) (AllowDisturbance :type Bool) (disturbance :type Double) (DisturbanceSubdivisionsDistance :type Double :default 4e+03))
  (inputs (PrimaryInput :type Any))
  (outputs (Areas :type Any) (Subdividing_splines :type Any))
  (bind AddTags_3 (Add%20Tags :TagsToAdd "Input:0"))
  (bind FilterDataByTag_0 (Filter%20Data%20By%20Tag :SelectedTags "Type:ClosedSpline"))
  ;[0 0 0 0.5 32] Inputs
  (bind Subgraph_0 (Subgraph :Graph /PCGPrimitives/Primitives/_Template_/Subgraphs/Primitives_InputsManagement.Primitives_InputsManagement :InputsFallbackGraph InputsFallbackGraph :InputsProcessingGraph InputsProcessingGraph))
  ;[0 0 0 0.5 32] Core Process
  (bind Subgraph_3 (Subgraph :Graph /PCGPrimitives/Primitives/Subdivide/Subgraphs/Subdivide_Areas_withGrid_CoreProcess.Subdivide_Areas_withGrid_CoreProcess :Allow_Disturbance AllowDisturbance :AngleJitter AngleJitter :Disturbance Disturbance :Disturbance_Subdivisions_distance DisturbanceSubdivisionsDistance :GridLength GridLength :GridWidth GridWidth :Subdiv_length SubdivLength :Subdiv_width SubdivWidth :SubgraphOverride-param CoreProcessGraph :Use_Subdivision_instead_of_sizes UseSubdivisionInsteadOfSizes))
  ;[0 0 0 0.5 32] Prepare Output
  (bind Subgraph_4 (Subgraph :Graph /PCGPrimitives/Primitives/Shared/PrepareOutputs/PrepareOutputs_ClosedSplinesAndSplines.PrepareOutputs_ClosedSplinesAndSplines :SubgraphOverride-param PrepareOutputsGraph))
  (connect DefaultInputNode PrimaryInput AddTags_3 In)
  (connect Subgraph_4 Out FilterDataByTag_0 In)
  (connect AddTags_3 Out Subgraph_0 In)
  (connect Subgraph_0 Out Subgraph_3 In)
  (connect Subgraph_3 Out Subgraph_4 In)
  (output Areas FilterDataByTag_0.InsideFilter)
  (output Subdividing_splines FilterDataByTag_0.OutsideFilter))
